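# Physics asset for the Lexa backpack cosmetic mesh (M_MED_Lexa_Pack), giving its two swords slight physical sway.
# A short capsule body represents the pack, and each of the left and right dynamic sword joints has a longer, thinner capsule.
# Both swords are constrained to the pack with identical, tight swing and twist limits, so they can only wobble a few degrees.
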PhysicsAsset /BRCosmetics/Accessories/FORT_Backpacks/Backpack_M_MED_Lexa/Meshes/M_MED_Lexa_Pack_Physics
bodies {
 dyn_L_sword_jnt [{shapeType=Capsule, length=15, radius=2, center=<-2.2482 0 0>, rotation=<rot 90 0 0>}]
 dyn_R_sword_jnt [{shapeType=Capsule, length=15, radius=2, center=<-2.2482 0 0>, rotation=<rot 90 0 0>}]
 pack [{shapeType=Capsule, length=6, radius=4}]
}
constraints [
 {bone1Name=dyn_L_sword_jnt, bone2Name=pack, swing1LimitDegrees=3.5, swing2LimitDegrees=2.2, twistLimitDegrees=3}
 {bone1Name=dyn_R_sword_jnt, bone2Name=pack, swing1LimitDegrees=3.5, swing2LimitDegrees=2.2, twistLimitDegrees=3}
]
