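# Niagara effect for the Lexa M backpack cosmetic: red glow sprites and short ribbon trails attached to skeletal mesh sockets or bones.
# Spawn groups select attachment points and mask which of each trail emitter's four ribbons stay visible.
# A system FN_Integrator value drives glow alpha and gameplay trails, which are also suppressed by skydiving or a vertical-velocity check.
# Separate _FE trail emitters spawn only in front-end preview, using randomized velocity.
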NiagaraSystem /BRCosmetics/Accessories/FORT_Backpacks/Backpack_M_MED_Lexa/VFX/NS_Backpack_Lexa_M
dependencies {
 usedDataInterfaces [&NiagaraDataInterfaceColorCurve, &NiagaraDataInterfaceCurve]
 usedDynamicInputs [
  /Niagara/DynamicInputs/UniformRange/UniformRangedVector4
  &Vector4FromCurve
  &Lerp_Float
  &ClampFloat
  &VectorLength
 ]
 usedModules [&SystemState, &FN_Integrator]
}
emitters {
 Glow {
  emitterUpdateScript [
   &EmitterState()
   &SpawnRate(SpawnRate=20, Use Spawn Probability=false)
  ]
  particleSpawnScript [
   &FN_SetSpawnGroup()
   /Niagara/Modules/Spawn/Initialization/InitializeParticle(
    Color=&MaskLinearColorBySpawnGroup(
     Masked Linear Color=<rgba 0.8698 0.0142 0 0.1>
     Passthrough Linear Color=&MultiplyLinearColor(
      Linear Color=<rgba 0.8698 0.0142 0 0.1>
      Scale Factor=&MakeFloat4FromVectorAndFloat("Vector (RGB)"=<1 1 1>, "Float (Alpha)"=$System.FN_Integrator.Value)
     )
    )
    Lifetime=&UniformRangedFloat(Minimum=1.5, Maximum=1.75)
    Module.Write Color=true
    Module.Write Lifetime=true
    Module.Write Mass=false
    Module.Write Material Random=true
    Module.Write Position=false
    Module.Write Scale=false
    Module.Write SpriteRotation=false
    Module.Write SpriteSize=true
    Position=&SimulationPosition()
    Sprite Rotation=&UniformRangedFloat(Minimum=-180, Maximum=180)
    Sprite Size=&Vector2DFromFloat(Value=&Multiply_Float(
      A=&MaskFloatBySpawnGroup(Masked Float=1, Passthrough Float=0.6)
      B=&RandomRangeFloat(Minimum=15, Maximum=20)
     ))
   )
   &CameraOffset(Camera Offset Amount=&RandomRangeFloat(Minimum=-5, Maximum=15))
   &SphereLocation(
    MaskSpawnGroup=false
    Override Seed=false
    Sphere Radius=0
    Surface Only=false
    UseNonUniformScale=false
    UseOffset=true
    UseSphereOrientation=false
   )
  ]
  particleUpdateScript [
   &ParticleState()
   &ScaleColor(
    Scale Alpha=&FloatFromCurve(FloatCurve=DataInterface() {propertyValues {Curve {keys [
         <at 0 0 RCIM_Cubic>
         {arriveTangent=0.1462, interpMode=RCIM_Cubic, leaveTangent=0.1462, time=0.169, value=1.0042}
         {arriveTangent=-1.2044, interpMode=RCIM_Cubic, leaveTangent=-1.2044, time=0.7656, value=0.1119}
         <at 1.0011 0.002>
        ]}}})
    ScaleA=true
    ScaleRGB=false
   ) {module=ScaleColor001}
   &ScaleSpriteSize(Scale Factor=&Vector2DFromFloat(Value=&FloatFromCurve(FloatCurve=DataInterface() {propertyValues {Curve {keys [
          {arriveTangent=7.5764, interpMode=RCIM_Cubic, leaveTangent=7.5765, tangentMode=RCTM_User, time=0, value=0.02}
          {arriveTangent=1.2063, interpMode=RCIM_Cubic, leaveTangent=1.2063, tangentMode=RCTM_User, time=0.18, value=0.99}
          <at 1 0.5 RCIM_Cubic>
         ]}}})))
   &SkeletalMeshLocation(
    Bone / Socket Index=&SpawnGroup()
    Inherit Velocity=false
    MaskSpawnGroup=false
    Mesh Sampling Type=::NewEnumerator5
    Skeletal Mesh=$User.SkeletalMesh
    Socket Sampling Mode=::NewEnumerator5
   )
  ]
  rendererClasses [&NiagaraSpriteRendererProperties]
  renderers [{rendererClass=&NiagaraSpriteRendererProperties}]
 }
 Trails_Left_Right {
  emitterUpdateScript [
   &EmitterState()
   &FN_SpawnRibbon_v2(
    Number of Ribbons=4
    SpawnRate:NiagaraFloat=&MakeCustomFloatFromBool(Bool=&SetBoolByFloatComparison(A=$System.FN_Integrator.Value, B=0), True Float=30, False Float=0)
   )
   &FN_SpawnRibbon_v2(Number of Ribbons=4, Spawn Mode=::NewEnumerator0, SpawnRate=30) {enabled=false, module=FN_SpawnRibbon_v2001}
  ]
  particleSpawnScript [
   &FN_SetSpawnGroup_v2(Number of Spawn Groups=4)
   /Game/Effects/Niagara/Modules/Ribbons/FN_InitializeRibbon_v4(
    Color:LinearColor=&MultiplyLinearColor(Linear Color=<rgba 0.8698 0.0142 0 0.65>, Scale Factor=<1 1 1 0.85>)
    Enable Skeletal Position=true
    Lifetime:NiagaraFloat=&MaskFloatBySpawnGroup(
     Masked Float=0
     Passthrough Float=&MaskFloatBySpawnGroup(Masked Float=0, Passthrough Float=0.16, Spawn Group Mask=3)
    )
    Ribbon Width=3
    SkeletalMesh:NiagaraDataInterfaceSkeletalMesh=$User.SkeletalMesh
   )
   &FN_FlipRibbonV()
   &AddVelocity(Velocity=&RotateVectorByQuaternion(Vector To Rotate=<0 -100 0>, Quaternion=$Particles.FN_UpdateRibbon_v5.SampledRotation))
  ]
  particleUpdateScript [
   /SetVariables(Particles.DownVelocityParameter:NiagaraFloat=&MakeCustomFloatFromBool(
     Bool=&SetBoolByFloatComparison(A=&MakeFloatFromVector(Vector=$Engine.Owner.Velocity, Channel=::NewEnumerator2), B=1000)
     True Float=0
     False Float=1
    )) {isSetParameters=true}
   &SolveForcesAndVelocity(Clamp Velocity=false, Limit Acceleration=false)
   /Game/Effects/Niagara/Modules/Ribbons/FN_UpdateRibbon_v5(Enable Skeletal Position=true, SkeletalMesh:NiagaraDataInterfaceSkeletalMesh=$User.SkeletalMesh)
   /SetVariables(Particles.Lifetime=&Multiply_Float(
     A=$Particles.Initial.Lifetime
     B=&Multiply_Float(
      A=$System.FN_Integrator.Value
      B=&Multiply_Float(
       A=&MakeCustomFloatFromBool(Bool=$User.IsSkydiving, True Float=0, False Float=1)
       B=$Particles.DownVelocityParameter
      )
     )
    )) {isSetParameters=true, module=SetVariables_#1}
   &ScaleColor(
    Scale Alpha=&FloatFromCurve(
     FloatCurve=DataInterface() {propertyValues {Curve {keys [<at 0.03 1>,<at 1 0>]}}}
     Scale Curve=&Multiply_Float(
      A=$System.FN_Integrator.Value
      B=&Multiply_Float(
       A=&MakeCustomFloatFromBool(Bool=$User.IsSkydiving, True Float=0, False Float=1)
       B=$Particles.DownVelocityParameter
      )
     )
    )
    ScaleA=true
    ScaleRGB=true
   ) {module=ScaleColor001}
   /SetVariables(Particles.RibbonFacing=&RotateVectorByQuaternion(Vector To Rotate=<0 1 0>, Quaternion=$Particles.FN_UpdateRibbon_v5.SampledRotation)) {isSetParameters=true, module=SetVariables_#2}
  ]
  rendererClasses [&NiagaraRibbonRendererProperties]
  renderers [{rendererClass=&NiagaraRibbonRendererProperties}]
 }
 Trails_Left_Right_FE {
  emitterUpdateScript [
   &EmitterState()
   &FN_SpawnRibbon_v2(
    Number of Ribbons=4
    SpawnRate:NiagaraFloat=&MakeCustomFloatFromBool(Bool=$User.bisFrontEndPreview, True Float=30, False Float=0)
   )
  ]
  particleSpawnScript [
   &FN_SetSpawnGroup_v2(Number of Spawn Groups=4)
   &FN_InitializeRibbon_v4(
    Color:LinearColor=&MultiplyLinearColor(Linear Color=<rgba 0.8698 0.0142 0 0.65>, Scale Factor=<1 1 1 0.85>)
    Enable Skeletal Position=true
    Lifetime:NiagaraFloat=&MaskFloatBySpawnGroup(
     Masked Float=0
     Passthrough Float=&MaskFloatBySpawnGroup(Masked Float=0, Passthrough Float=0.14, Spawn Group Mask=3)
    )
    Ribbon Width=3
    SkeletalMesh:NiagaraDataInterfaceSkeletalMesh=$User.SkeletalMesh
   )
   &FN_FlipRibbonV()
   &AddVelocity(Velocity=&RotateVectorByQuaternion(
     Vector To Rotate=&RandomRangeVector(Minimum=<-4 -120 -4>, Maximum=<4 -180 4>)
     Quaternion=$Particles.FN_UpdateRibbon_v5.SampledRotation
    ))
  ]
  particleUpdateScript [
   &SolveForcesAndVelocity(Clamp Velocity=false, Limit Acceleration=false)
   &FN_UpdateRibbon_v5(Enable Skeletal Position=true, SkeletalMesh:NiagaraDataInterfaceSkeletalMesh=$User.SkeletalMesh)
   &ScaleColor(
    Scale Alpha=&FloatFromCurve(FloatCurve=DataInterface() {propertyValues {Curve {keys [<at 0.03 1>,<at 1 0>]}}})
    ScaleA=true
    ScaleRGB=true
   ) {module=ScaleColor001}
   /SetVariables(Particles.RibbonFacing=&RotateVectorByQuaternion(Vector To Rotate=<0 1 0>, Quaternion=$Particles.FN_UpdateRibbon_v5.SampledRotation)) {isSetParameters=true, module=SetVariables_#2}
  ]
  rendererClasses [&NiagaraRibbonRendererProperties]
  renderers [{rendererClass=&NiagaraRibbonRendererProperties}]
 }
 Trails_Top_Down {
  emitterUpdateScript [
   &EmitterState()
   &FN_SpawnRibbon_v2(
    Number of Ribbons=4
    SpawnRate:NiagaraFloat=&MakeCustomFloatFromBool(Bool=&SetBoolByFloatComparison(A=$System.FN_Integrator.Value, B=0), True Float=30, False Float=0)
   )
  ]
  particleSpawnScript [
   &FN_SetSpawnGroup_v2(Number of Spawn Groups=4)
   &FN_InitializeRibbon_v4(
    Color:LinearColor=&MultiplyLinearColor(Linear Color=<rgba 0.8698 0.0142 0 0.65>, Scale Factor=<1 1 1 0.85>)
    Enable Skeletal Position=true
    Lifetime:NiagaraFloat=&MaskFloatBySpawnGroup(
     Masked Float=0
     Passthrough Float=&MaskFloatBySpawnGroup(
      Masked Float=0
      Passthrough Float=&MaskFloatBySpawnGroup(Masked Float=0, Passthrough Float=0.14)
      Spawn Group Mask=2
     )
     Spawn Group Mask=1
    )
    Ribbon Width:NiagaraFloat=&MaskFloatBySpawnGroup(Masked Float=6, Passthrough Float=1.2)
    SkeletalMesh:NiagaraDataInterfaceSkeletalMesh=$User.SkeletalMesh
   )
   &FN_FlipRibbonV()
   &AddVelocity(Velocity=&RotateVectorByQuaternion(Vector To Rotate=<0 -100 0>, Quaternion=$Particles.FN_UpdateRibbon_v5.SampledRotation))
  ]
  particleUpdateScript [
   /SetVariables(Particles.DownVelocityParameter:NiagaraFloat=&MakeCustomFloatFromBool(
     Bool=&SetBoolByFloatComparison(A=&MakeFloatFromVector(Vector=$Engine.Owner.Velocity, Channel=::NewEnumerator2), B=1000)
     True Float=0
     False Float=1
    )) {isSetParameters=true}
   &SolveForcesAndVelocity(Clamp Velocity=false, Limit Acceleration=false)
   &FN_UpdateRibbon_v5(Enable Skeletal Position=true, SkeletalMesh:NiagaraDataInterfaceSkeletalMesh=$User.SkeletalMesh)
   /SetVariables(Particles.Lifetime=&Multiply_Float(
     A=$Particles.Initial.Lifetime
     B=&Multiply_Float(
      A=$System.FN_Integrator.Value
      B=&Multiply_Float(
       A=&MakeCustomFloatFromBool(Bool=$User.IsSkydiving, True Float=0, False Float=1)
       B=$Particles.DownVelocityParameter
      )
     )
    )) {isSetParameters=true, module=SetVariables_#1}
   &ScaleColor(
    Scale Alpha=&FloatFromCurve(
     FloatCurve=DataInterface() {propertyValues {Curve {keys [<at 0.03 1>,<at 1 0>]}}}
     Scale Curve=&Multiply_Float(
      A=$System.FN_Integrator.Value
      B=&Multiply_Float(
       A=&MakeCustomFloatFromBool(Bool=$User.IsSkydiving, True Float=0, False Float=1)
       B=$Particles.DownVelocityParameter
      )
     )
    )
    ScaleA=true
    ScaleRGB=true
   ) {module=ScaleColor001}
  ]
  rendererClasses [&NiagaraRibbonRendererProperties]
  renderers [{rendererClass=&NiagaraRibbonRendererProperties}]
 }
 Trails_Top_FE {
  emitterUpdateScript [
   &EmitterState()
   &FN_SpawnRibbon_v2(
    Number of Ribbons=4
    SpawnRate:NiagaraFloat=&MakeCustomFloatFromBool(Bool=$User.bisFrontEndPreview, True Float=30, False Float=0)
   )
  ]
  particleSpawnScript [
   &FN_SetSpawnGroup_v2(Number of Spawn Groups=4)
   &FN_InitializeRibbon_v4(
    Color:LinearColor=&MultiplyLinearColor(Linear Color=<rgba 0.8698 0.0142 0 0.65>, Scale Factor=<1 1 1 0.85>)
    Enable Skeletal Position=true
    Lifetime:NiagaraFloat=&MaskFloatBySpawnGroup(
     Masked Float=0
     Passthrough Float=&MaskFloatBySpawnGroup(
      Masked Float=0
      Passthrough Float=&MaskFloatBySpawnGroup(Masked Float=0, Passthrough Float=0.12)
      Spawn Group Mask=2
     )
     Spawn Group Mask=1
    )
    Ribbon Width:NiagaraFloat=&MaskFloatBySpawnGroup(Masked Float=6, Passthrough Float=1.2)
    SkeletalMesh:NiagaraDataInterfaceSkeletalMesh=$User.SkeletalMesh
   )
   &FN_FlipRibbonV()
   &AddVelocity(Velocity=&RotateVectorByQuaternion(
     Vector To Rotate=&RandomRangeVector(Minimum=<-4 -80 -4>, Maximum=<4 -120 4>)
     Quaternion=$Particles.FN_UpdateRibbon_v5.SampledRotation
    )) {module=AddVelocity001}
  ]
  particleUpdateScript [
   &SolveForcesAndVelocity(Clamp Velocity=false, Limit Acceleration=false)
   &FN_UpdateRibbon_v5(Enable Skeletal Position=true, SkeletalMesh:NiagaraDataInterfaceSkeletalMesh=$User.SkeletalMesh)
   &ScaleColor(
    Scale Alpha=&FloatFromCurve(FloatCurve=DataInterface() {propertyValues {Curve {keys [<at 0.03 1>,<at 1 0>]}}})
    ScaleA=true
    ScaleRGB=true
   ) {module=ScaleColor001}
  ]
  rendererClasses [&NiagaraRibbonRendererProperties]
  renderers [{rendererClass=&NiagaraRibbonRendererProperties}]
 }
}
summary {
 systemName=NS_Backpack_Lexa_M
 emitters [
  {emitterName=Glow, rendererClasses [&NiagaraSpriteRendererProperties]}
  {emitterName=Trails_Left_Right, rendererClasses [&NiagaraRibbonRendererProperties]}
  {emitterName=Trails_Top_Down, rendererClasses [&NiagaraRibbonRendererProperties]}
  {emitterName=Trails_Left_Right_FE, rendererClasses [&NiagaraRibbonRendererProperties]}
  {emitterName=Trails_Top_FE, rendererClasses [&NiagaraRibbonRendererProperties]}
 ]
 userVariables [
  User.bIsFrontEndPreview {type=NiagaraBool}
  User.ColorChange {type=LinearColor, description="", defaultValue {struct=LinearColor, value=<rgba 1 1 1 1>}}
  User.IsSkydiving {type=NiagaraBool}
  User.RibbonIDOffset {type=NiagaraInt32}
  User.SkeletalMesh {
   type=NiagaraDataInterfaceSkeletalMesh
   defaultValue {
    struct=NiagaraExt_VariableValue_DataInterface
    value {dataInterface=&self:NiagaraDataInterfaceSkeletalMesh_1}
   }
  }
 ]
}
